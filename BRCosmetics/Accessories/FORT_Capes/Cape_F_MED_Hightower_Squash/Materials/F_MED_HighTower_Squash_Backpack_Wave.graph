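# Masked cloth material for the Hightower Squash cape backpack, with animated waving and timed dissolve.
# Character PreFX feeds metals LUT, silk and optional holo-foil layers, then PostFX, with a switchable second silk pass.
# World position offset adds time-scrolled curl noise over pre-skinned positions, limited by an FX-texture channel mask and WaveStrength.
# Dithered opacity mask combines the FX texture's blue channel with an in/out transition driven by custom primitive data timestamps.
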Material /BRCosmetics/Accessories/FORT_Capes/Cape_F_MED_Hightower_Squash/Materials/F_MED_HighTower_Squash_Backpack_Wave
graph {
 properties {
  bAutomaticallySetUsageInEditor=false
  bUseMaterialAttributes=true
  bUsedWithMeshDeformer=true
  bUsedWithSkeletalMesh=true
  ditherOpacityMask=true
  opacityMaskClipValue=0.5
  pixelDepthOffsetMode=PDOM_Legacy
  usedShadingModels=Default Lit
 }
 nodes [
  ScalarParameter() {defaultValue=1, parameterName=DEBUG}
  n1 = TextureSampleParameter2D() {
   group=Skin FX
   parameterName=SkinFX_Mask
   samplerType=SAMPLERTYPE_LinearColor
   texture=/BRCosmetics/Accessories/FORT_Capes/Cape_F_MED_Hightower_Squash/Textures/T_F_MED_HighTower_Squash_Backpack_FX
  }
  n2 = MaterialFunctionCall() {materialFunction=&BaseFN_CharacterFunction_PreFX}
  n3 = PreSkinnedPosition()
  n4 = VertexInterpolator(VS=@n3)
  n5 = VertexInterpolator(VS=PreSkinnedNormal())
  n6 = AppendVector(A=@n1.RGB, B=@n1.RGB)
  n7 = MaterialFunctionCall!(
   Alpha=ChannelMaskParameter(None=@n6) {group=Skin FX, parameterName=Silk Channel}.R
   Material=MaterialFunctionCall(Mask=@n6, Materials=@n2.Result) {materialFunction=&MF_MetalsLUT}.Result
  ) {materialFunction=/Game/Weapons/WeaponSkins/MaterialFunction/MF_Silk}
  n8 = MaterialFunctionCall(
   AO=@n2.AO
   Distance Blend=@n2.AO
   Fresnel=@n2.AO
   Material=StaticSwitchParameter(
    True=MaterialFunctionCall!(PreskinnedLocalPosition=@n4.PS, PreskinnedLocalNormal=@n5.PS, Material=@n7.Result) {materialFunction=/Game/Characters/Player/Female/Medium/Bodies/F_MED_HighTower_Squash/Materials/MF_HoloFoil}.Result
    False=@n7.Result
   ) {parameterName=BUseHoloFoil}
  ) {materialFunction=&BaseFN_CharacterFunction_PostFX}
  n9 = GetMaterialAttributes(None=@n8.Result) {attributeGetTypes [#1]}
  n10 = ChannelMaskParameter(None=@n6) {group=Skin FX, maskChannel=Green, parameterName=WaveMask, defaultValue=<rgba 0 1 0 0>}
  n11 = Time()
  n12 = Constant3Vector() {constant=<rgba 0 50 0 0>}
  n13 = Divide(A=Add(A=@n3, B=Multiply(A=@n11, B=@n12)), B=Constant3Vector() {constant=<rgba 160 75 160 0>})
  n14 = Multiply(
   A=ConstantBiasScale(None=VectorNoise(World Position=@n13) {noiseFunction=VNF_CurlALU, tileSize=800}) {bias=3}
   B=Transform(None=VectorParameter() {parameterName=WaveStrength, defaultValue=<rgba 0 -1.5 0 1>}.RGB) {transformSourceType=TRANSFORMSOURCE_Local}
  )
  n15 = SetMaterialAttributes!(MaterialAttributes=@n9.MaterialAttributes, World Position Offset=Add(A=@n9.World Position Offset, B=Multiply(A=@n10.R, B=@n14))) {attributeSetTypes [#1]}
  n16 = TextureSampleParameter2D() {
   group=Skin FX
   parameterName=SkinFX_Mask
   samplerType=SAMPLERTYPE_LinearColor
   texture=&T_F_MED_HighTower_Squash_Backpack_FX
  }
  n17 = ConstantBiasScale(None=MaterialFunctionCall(
    TimestampA=ScalarParameter() {bUseCustomPrimitiveData=true, parameterName=TimestampA}
    TransitionLength_In=ScalarParameter() {defaultValue=0.75, parameterName=TransitionLength_In}
    TimestampB=ScalarParameter() {bUseCustomPrimitiveData=true, parameterName=TimestampB, primitiveDataIndex=1}
    TransitionLength_Out=ScalarParameter() {defaultValue=0.7, parameterName=TransitionLength_Out}
    Transition In Only?=StaticBool()
   ) {materialFunction=&MF_Timestamp_InOut_01}.Linear) {bias=-0.5, scale=2}
  SetMaterialAttributes!(
   MaterialAttributes=StaticSwitchParameter(
    True=MaterialFunctionCall(Alpha=@n1.R, Material=@n15) {materialFunction=&MF_Silk}.Result
    False=@n15
   ) {parameterName=UseSilk}
   Opacity Mask=Power(Base=Add(A=@n16.B, B=@n17)) {constExponent=4}
  ) {attributeSetTypes [#2]}
  Sine(None=Time()) {period=4}
 ]
}
