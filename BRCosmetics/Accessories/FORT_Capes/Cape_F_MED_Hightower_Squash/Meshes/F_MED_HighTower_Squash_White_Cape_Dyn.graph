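# Physics asset for the white Hightower Squash cape on the medium female body, defining bodies for dynamic cape motion.
# Five box-body chains (left, right, A left, A right, middle) hang from tiny MSTR anchor capsules joined to spine_05, with swing limits and tight twist.
# Capsules on pelvis, spine, thighs and calves form the body, linked by angularly free constraints.
# Cross-links angularly lock both A-chain first segments to the middle chain's third segment.
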PhysicsAsset /BRCosmetics/Accessories/FORT_Capes/Cape_F_MED_Hightower_Squash/Meshes/F_MED_HighTower_Squash_White_Cape_Dyn
bodies {
 calf_l [{shapeType=Capsule, length=32.8125, radius=6.9688, center=<-22.7514 0.0005 0>, rotation=<rot 90 0 0>}]
 calf_r [{shapeType=Capsule, length=32.8125, radius=6.9688, center=<22.7514 0.0005 0>, rotation=<rot 90 0 0>}]
 dyn_cape_01_l [{shapeType=Box, extentX=27, extentY=65.375, extentZ=11.0625, center=<13 0 -0.0001>}]
 dyn_cape_01_l_MSTR [{shapeType=Capsule, length=0.505, radius=0.505}]
 dyn_cape_01_r [{shapeType=Box, extentX=27, extentY=67.1875, extentZ=11.0625, center=<13 0 -0.0001>}]
 dyn_cape_01_r_MSTR [{shapeType=Capsule, length=0.505, radius=0.505}]
 dyn_cape_02_l [{shapeType=Box, extentX=27, extentY=65.375, extentZ=11.0625, center=<13 0 -0.0001>}]
 dyn_cape_02_r [{shapeType=Box, extentX=27, extentY=67.1875, extentZ=11.0625, center=<13 0 -0.0001>}]
 dyn_cape_03_l [{shapeType=Box, extentX=27, extentY=65.375, extentZ=11.0625, center=<13 0 -0.0001>}]
 dyn_cape_03_r [{shapeType=Box, extentX=27, extentY=67.1875, extentZ=11.0625, center=<13 0 -0.0001>}]
 dyn_cape_04_l [{shapeType=Box, extentX=27, extentY=65.375, extentZ=11.0625, center=<13 0 -0.0001>}]
 dyn_cape_04_r [{shapeType=Box, extentX=27, extentY=67.1875, extentZ=11.0625, center=<13 0 -0.0001>}]
 dyn_cape_A_01_l [{shapeType=Box, extentX=25.9375, extentY=57.25, extentZ=10.8125, center=<12 0 -0.0001>}]
 dyn_cape_A_01_l_MSTR [{shapeType=Capsule, length=0.505, radius=0.505}]
 dyn_cape_A_01_r [{shapeType=Box, extentX=25.9375, extentY=57.25, extentZ=10.8125, center=<12 0 -0.0001>}]
 dyn_cape_A_01_r_MSTR [{shapeType=Capsule, length=0.505, radius=0.505}]
 dyn_cape_A_02_l [{shapeType=Box, extentX=25.9375, extentY=57.25, extentZ=10.8125, center=<12 0 -0.0001>}]
 dyn_cape_A_02_r [{shapeType=Box, extentX=25.9375, extentY=57.25, extentZ=10.8125, center=<12 0 -0.0001>}]
 dyn_cape_A_03_l [{shapeType=Box, extentX=25.9375, extentY=57.25, extentZ=10.8125, center=<12 0 -0.0001>}]
 dyn_cape_A_03_r [{shapeType=Box, extentX=25.9375, extentY=57.25, extentZ=10.8125, center=<12 0 -0.0001>}]
 dyn_cape_A_04_l [{shapeType=Box, extentX=25.9375, extentY=57.25, extentZ=10.8125, center=<12 0 -0.0001>}]
 dyn_cape_A_04_r [{shapeType=Box, extentX=25.9375, extentY=57.25, extentZ=10.8125, center=<12 0 -0.0001>}]
 dyn_cape_mid_01 [{shapeType=Box, extentX=13.4375, extentY=18.8125, extentZ=2.1875, center=<6.342 0 -0.0001>}]
 dyn_cape_mid_01_MSTR [{shapeType=Capsule, length=0.505, radius=0.505}]
 dyn_cape_mid_02 [{shapeType=Box, extentX=13.4375, extentY=18.8125, extentZ=2.1875, center=<6.342 0 -0.0001>}]
 dyn_cape_mid_03 [{shapeType=Box, extentX=13.4375, extentY=18.8125, extentZ=2.1875, center=<6.342 0 -0.0001>}]
 dyn_cape_mid_04 [{shapeType=Box, extentX=13.4375, extentY=18.8125, extentZ=2.1875, center=<6.342 0 -0.0001>}]
 dyn_cape_mid_05 [{shapeType=Box, extentX=13.4375, extentY=18.8125, extentZ=2.1875, center=<6.342 0 -0.0001>}]
 pelvis [{shapeType=Capsule, length=13.4375, radius=10.125, center=<-5.0792 -2.7401 0>}]
 spine_01 [{shapeType=Capsule, length=7.9375, radius=8.75, center=<2.8713 0.089 -0.0001>}]
 spine_02 [{shapeType=Capsule, length=5.9062, radius=10.7812, center=<2.8713 2.6773 -0.0003>}]
 spine_03 [{shapeType=Capsule, length=5.7188, radius=10.9688, center=<2.807 2.666 0>}]
 spine_04 [{shapeType=Capsule, length=7.3125, radius=9.375, center=<2.8713 1.1783 0>}]
 spine_05 [{shapeType=Capsule, length=8.6875, radius=8, center=<2.8713 0 0>}]
 thigh_l [{shapeType=Capsule, length=31.4375, radius=8.3438, center=<-22.7514 0.0005 0>, rotation=<rot 90 0 0>}]
 thigh_r [{shapeType=Capsule, length=31.4375, radius=8.3438, center=<22.7514 -0.0005 0>, rotation=<rot 90 0 0>}]
}
constraints [
 {bone1Name=dyn_cape_01_r, bone2Name=dyn_cape_01_r_MSTR, swing1LimitDegrees=80, swing2LimitDegrees=80, twistLimitDegrees=5}
 {bone1Name=dyn_cape_02_r, bone2Name=dyn_cape_01_r, swing1LimitDegrees=45, swing2LimitDegrees=45, twistLimitDegrees=5}
 {bone1Name=dyn_cape_03_r, bone2Name=dyn_cape_02_r, swing1LimitDegrees=45, swing2LimitDegrees=45, twistLimitDegrees=5}
 {bone1Name=dyn_cape_04_r, bone2Name=dyn_cape_03_r, swing1LimitDegrees=45, swing2LimitDegrees=45, twistLimitDegrees=5}
 {bone1Name=dyn_cape_A_01_r, bone2Name=dyn_cape_A_01_r_MSTR, swing1LimitDegrees=80, swing2LimitDegrees=80, twistLimitDegrees=5}
 {bone1Name=dyn_cape_A_02_r, bone2Name=dyn_cape_A_01_r, swing1LimitDegrees=45, swing2LimitDegrees=45, twistLimitDegrees=5}
 {bone1Name=dyn_cape_A_03_r, bone2Name=dyn_cape_A_02_r, swing1LimitDegrees=45, swing2LimitDegrees=45, twistLimitDegrees=5}
 {bone1Name=dyn_cape_A_04_r, bone2Name=dyn_cape_A_03_r, swing1LimitDegrees=45, swing2LimitDegrees=45, twistLimitDegrees=5}
 {bone1Name=dyn_cape_A_01_l, bone2Name=dyn_cape_A_01_l_MSTR, swing1LimitDegrees=80, swing2LimitDegrees=80, twistLimitDegrees=5}
 {bone1Name=dyn_cape_A_02_l, bone2Name=dyn_cape_A_01_l, swing1LimitDegrees=45, swing2LimitDegrees=45, twistLimitDegrees=5}
 {bone1Name=dyn_cape_A_03_l, bone2Name=dyn_cape_A_02_l, swing1LimitDegrees=45, swing2LimitDegrees=45, twistLimitDegrees=5}
 {bone1Name=dyn_cape_A_04_l, bone2Name=dyn_cape_A_03_l, swing1LimitDegrees=45, swing2LimitDegrees=45, twistLimitDegrees=5}
 {bone1Name=dyn_cape_01_l, bone2Name=dyn_cape_01_l_MSTR, swing1LimitDegrees=80, swing2LimitDegrees=80, twistLimitDegrees=5}
 {bone1Name=dyn_cape_02_l, bone2Name=dyn_cape_01_l, swing1LimitDegrees=45, swing2LimitDegrees=45, twistLimitDegrees=5}
 {bone1Name=dyn_cape_03_l, bone2Name=dyn_cape_02_l, swing1LimitDegrees=45, swing2LimitDegrees=45, twistLimitDegrees=5}
 {bone1Name=dyn_cape_04_l, bone2Name=dyn_cape_03_l, swing1LimitDegrees=45, swing2LimitDegrees=45, twistLimitDegrees=5}
 {bone1Name=dyn_cape_mid_01, bone2Name=dyn_cape_mid_01_MSTR, swing1LimitDegrees=45, swing2LimitDegrees=45, twistLimitDegrees=5}
 {bone1Name=dyn_cape_mid_02, bone2Name=dyn_cape_mid_01, swing1LimitDegrees=45, swing2LimitDegrees=45, twistLimitDegrees=5}
 {bone1Name=dyn_cape_mid_03, bone2Name=dyn_cape_mid_02, swing1LimitDegrees=45, swing2LimitDegrees=45, twistLimitDegrees=5}
 {bone1Name=dyn_cape_mid_04, bone2Name=dyn_cape_mid_03, swing1LimitDegrees=45, swing2LimitDegrees=45, twistLimitDegrees=5}
 {bone1Name=dyn_cape_mid_05, bone2Name=dyn_cape_mid_04, swing1LimitDegrees=45, swing2LimitDegrees=45, twistLimitDegrees=5}
 {
  bone1Name=dyn_cape_01_r_MSTR
  bone2Name=spine_05
  swing1LimitDegrees=45
  swing1Motion=Free
  swing2LimitDegrees=45
  swing2Motion=Free
  twistLimitDegrees=45
  twistMotion=Free
 }
 {
  bone1Name=dyn_cape_A_01_r_MSTR
  bone2Name=spine_05
  swing1LimitDegrees=45
  swing1Motion=Free
  swing2LimitDegrees=45
  swing2Motion=Free
  twistLimitDegrees=45
  twistMotion=Free
 }
 {
  bone1Name=dyn_cape_A_01_l_MSTR
  bone2Name=spine_05
  swing1LimitDegrees=45
  swing1Motion=Free
  swing2LimitDegrees=45
  swing2Motion=Free
  twistLimitDegrees=45
  twistMotion=Free
 }
 {
  bone1Name=dyn_cape_01_l_MSTR
  bone2Name=spine_05
  swing1LimitDegrees=45
  swing1Motion=Free
  swing2LimitDegrees=45
  swing2Motion=Free
  twistLimitDegrees=45
  twistMotion=Free
 }
 {
  bone1Name=dyn_cape_mid_01_MSTR
  bone2Name=spine_05
  swing1LimitDegrees=45
  swing1Motion=Free
  swing2LimitDegrees=45
  swing2Motion=Free
  twistLimitDegrees=45
  twistMotion=Free
 }
 {
  bone1Name=spine_05
  bone2Name=spine_04
  swing1LimitDegrees=45
  swing1Motion=Free
  swing2LimitDegrees=45
  swing2Motion=Free
  twistLimitDegrees=45
  twistMotion=Free
 }
 {
  bone1Name=spine_02
  bone2Name=spine_01
  swing1LimitDegrees=45
  swing1Motion=Free
  swing2LimitDegrees=45
  swing2Motion=Free
  twistLimitDegrees=45
  twistMotion=Free
 }
 {
  bone1Name=spine_03
  bone2Name=spine_02
  swing1LimitDegrees=45
  swing1Motion=Free
  swing2LimitDegrees=45
  swing2Motion=Free
  twistLimitDegrees=45
  twistMotion=Free
 }
 {
  bone1Name=spine_04
  bone2Name=spine_03
  swing1LimitDegrees=45
  swing1Motion=Free
  swing2LimitDegrees=45
  swing2Motion=Free
  twistLimitDegrees=45
  twistMotion=Free
 }
 {
  bone1Name=spine_01
  bone2Name=pelvis
  swing1LimitDegrees=45
  swing1Motion=Free
  swing2LimitDegrees=45
  swing2Motion=Free
  twistLimitDegrees=45
  twistMotion=Free
 }
 {
  bone1Name=thigh_l
  bone2Name=pelvis
  swing1LimitDegrees=45
  swing1Motion=Free
  swing2LimitDegrees=45
  swing2Motion=Free
  twistLimitDegrees=45
  twistMotion=Free
 }
 {
  bone1Name=calf_l
  bone2Name=thigh_l
  swing1LimitDegrees=45
  swing1Motion=Free
  swing2LimitDegrees=45
  swing2Motion=Free
  twistLimitDegrees=45
  twistMotion=Free
 }
 {
  bone1Name=thigh_r
  bone2Name=pelvis
  swing1LimitDegrees=45
  swing1Motion=Free
  swing2LimitDegrees=45
  swing2Motion=Free
  twistLimitDegrees=45
  twistMotion=Free
 }
 {
  bone1Name=calf_r
  bone2Name=thigh_r
  swing1LimitDegrees=45
  swing1Motion=Free
  swing2LimitDegrees=45
  swing2Motion=Free
  twistLimitDegrees=45
  twistMotion=Free
 }
 {
  bone1Name=dyn_cape_A_01_r
  bone2Name=dyn_cape_mid_03
  swing1LimitDegrees=45
  swing1Motion=Locked
  swing2LimitDegrees=45
  swing2Motion=Locked
  twistLimitDegrees=45
  twistMotion=Locked
 }
 {
  bone1Name=dyn_cape_A_01_l
  bone2Name=dyn_cape_mid_03
  swing1LimitDegrees=45
  swing1Motion=Locked
  swing2LimitDegrees=45
  swing2Motion=Locked
  twistLimitDegrees=45
  twistMotion=Locked
 }
]
