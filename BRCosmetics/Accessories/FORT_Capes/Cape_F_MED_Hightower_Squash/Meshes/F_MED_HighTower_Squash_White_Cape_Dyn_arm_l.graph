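# Physics asset for the left-arm dynamic piece of the white Hightower Squash cape, a cosmetic accessory mesh.
# Three identical flat box bodies form a constrained chain hung from a tiny master capsule on the left forearm.
# The root link has wide swing limits and free twist, while lower links have tighter swing and very narrow twist.
# Forearm, thigh and calf capsules represent the body, apparently as colliders for the cape.
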PhysicsAsset /BRCosmetics/Accessories/FORT_Capes/Cape_F_MED_Hightower_Squash/Meshes/F_MED_HighTower_Squash_White_Cape_Dyn_arm_l
bodies {
 calf_l [{shapeType=Capsule, length=32.8125, radius=6.9688, center=<-22.7514 0.0005 0>, rotation=<rot 90 0 0>}]
 dyn_cape_01_prox_l [{shapeType=Box, extentX=13.4375, extentY=18.8125, extentZ=2.1875, center=<6.342 0 -0.0001>}]
 dyn_cape_01_prox_l_MSTR [{shapeType=Capsule, length=0.505, radius=0.505}]
 dyn_cape_02_prox_l [{shapeType=Box, extentX=13.4375, extentY=18.8125, extentZ=2.1875, center=<6.342 0 -0.0001>}]
 dyn_cape_03_prox_l [{shapeType=Box, extentX=13.4375, extentY=18.8125, extentZ=2.1875, center=<6.342 0 -0.0001>}]
 lowerarm_l [{shapeType=Capsule, length=18.8125, radius=3.875, center=<6.2942 0 0>, rotation=<rot 90 0 0>}]
 thigh_l [{shapeType=Capsule, length=31.4375, radius=8.3438, center=<-22.7514 0.0005 0>, rotation=<rot 90 0 0>}]
}
constraints [
 {
  bone1Name=dyn_cape_01_prox_l
  bone2Name=dyn_cape_01_prox_l_MSTR
  swing1LimitDegrees=100
  swing2LimitDegrees=120
  twistLimitDegrees=30
  twistMotion=Free
 }
 {bone1Name=dyn_cape_02_prox_l, bone2Name=dyn_cape_01_prox_l, swing1LimitDegrees=45, swing2LimitDegrees=45, twistLimitDegrees=5}
 {bone1Name=dyn_cape_03_prox_l, bone2Name=dyn_cape_02_prox_l, swing1LimitDegrees=45, swing2LimitDegrees=45, twistLimitDegrees=5}
 {
  bone1Name=dyn_cape_01_prox_l_MSTR
  bone2Name=lowerarm_l
  swing1LimitDegrees=45
  swing1Motion=Free
  swing2LimitDegrees=45
  swing2Motion=Free
  twistLimitDegrees=45
  twistMotion=Free
 }
 {
  bone1Name=calf_l
  bone2Name=thigh_l
  swing1LimitDegrees=45
  swing1Motion=Free
  swing2LimitDegrees=45
  swing2Motion=Free
  twistLimitDegrees=45
  twistMotion=Free
 }
]
