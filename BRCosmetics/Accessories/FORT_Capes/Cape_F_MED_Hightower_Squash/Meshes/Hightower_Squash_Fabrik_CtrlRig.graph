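# Control Rig for the Hightower Squash cape that aims and bends its left and right sides toward attach bones on the wrists.
# offset_l and offset_r weight pole-bone translations, which serve as secondary targets for aiming each side's cape root at its wrist.
# FABRIK then pulls bone chains to the attach points, though mid and bottom chains are absent from this hierarchy.
# Right top and mid solvers, bone-transform offsets and some vector math sit unconnected.
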ControlRigBlueprint /BRCosmetics/Accessories/FORT_Capes/Cape_F_MED_Hightower_Squash/Meshes/Hightower_Squash_Fabrik_CtrlRig
graphs [
 RigVMFunctionLibrary {links [], nodes []}
 RigVMModel {nodes [
   /Script/ControlRig.RigUnit_AddBoneTransform(
    Bone=cape_top_r_1
    ExecutePin=()
    Transform=(Rotation=(X=0,Y=0,Z=0,W=1),Translation=(X=0,Y=0,Z=0),Scale3D=(X=0.1,Y=0.1,Z=0.1))
    Weight=1
    bPostMultiply=false
    bPropagateToChildren=True
   ) {is=struct RigVMUnitNode}
   > &RigUnit_AddBoneTransform(
    Bone=cape_mid_r_1
    Transform=(Rotation=(X=0,Y=0,Z=0,W=1),Translation=(X=0,Y=0,Z=0),Scale3D=(X=0.1,Y=0.1,Z=0.1))
    Weight=1
    bPostMultiply=false
    bPropagateToChildren=True
   )
   > &RigUnit_AddBoneTransform(
    Bone=cape_bot_r_1
    Transform=(Rotation=(X=0,Y=0,Z=0,W=1),Translation=(X=0,Y=0,Z=0),Scale3D=(X=0.1,Y=0.1,Z=0.1))
    Weight=1
    bPostMultiply=false
    bPropagateToChildren=True
   )
   RigUnit_BeginExecution = &RigUnit_BeginExecution()
   n1 = /Script/ControlRig.RigUnit_BoneName(Bone=cape_pv_up_l) {is=struct RigVMUnitNode}
   /Script/ControlRig.RigUnit_SetBoneTranslation(
    Translation=(X=33,Y=-14,Z=0)
    bPropagateToChildren=True
    Bone=@n1.Bone
    Weight=&RigVMFunction_MathFloatAdd(A=0, B=RigVMVariableNode() {variable=offset_l}.Value) {template="Add::Execute(in A,in B,out Result)"}.Result
    ExecutePin=@RigUnit_BeginExecution.ExecutePin
   ) {is=struct RigVMUnitNode}
   n2 = &RigUnit_BoneName(Bone=cape_pv_aim_l)
   n3 = &RigUnit_GetBoneTransform(Bone=@n1.Bone)
   > /Script/ControlRig.RigUnit_AimBone(
    DebugSettings=(bEnabled=true,Scale=10,WorldOffset=(Rotation=(X=0,Y=0,Z=0,W=1),Translation=(X=0,Y=0,Z=0),Scale3D=(X=1,Y=1,Z=1)))
    Primary=(Weight=1,Axis=(X=1,Y=0,Z=0),Target=(X=1,Y=0,Z=0),Kind=Location)
    Secondary=(Weight=1,Axis=(X=0,Y=1,Z=0),Target=(X=0,Y=0,Z=0),Kind=Location)
    Weight=1
    bPropagateToChildren=True
    Bone=@n2.Bone
    Primary.Target=&RigUnit_GetBoneTransform(Bone=&RigUnit_BoneName(Bone=attach_l).Bone).Transform.Translation
    Secondary.Target=@n3.Transform.Translation
   ) {is=struct RigVMUnitNode}
   n4 = &RigUnit_BoneName(Bone=cape_pv_up_r)
   > &RigUnit_SetBoneTranslation(
    Translation=(X=-33,Y=14,Z=0)
    bPropagateToChildren=True
    Bone=@n4.Bone
    Weight=&RigVMFunction_MathFloatAdd(A=0, B=RigVMVariableNode() {variable=offset_r}.Value) {template="Add::Execute(in A,in B,out Result)"}.Result
   )
   > &RigUnit_AimBone(
    Bone=cape_pv_aim_r
    DebugSettings=(bEnabled=true,Scale=10,WorldOffset=(Rotation=(X=0,Y=0,Z=0,W=1),Translation=(X=0,Y=0,Z=0),Scale3D=(X=1,Y=1,Z=1)))
    Primary=(Weight=1,Axis=(X=1,Y=0,Z=0),Target=(X=1,Y=0,Z=0),Kind=Location)
    Secondary=(Weight=1,Axis=(X=0,Y=-1,Z=0),Target=(X=1,Y=0,Z=0),Kind=Location)
    Weight=1
    bPropagateToChildren=True
    Secondary.Target=&RigUnit_GetBoneTransform(Bone=@n4.Bone).Transform.Translation
    Primary.Target=&RigUnit_GetBoneTransform(Bone=&RigUnit_BoneName(Bone=attach_r).Bone).Transform.Translation
   )
   RigUnit_GetBoneTransform = &RigUnit_GetBoneTransform(Bone=attach_l)
   > /Script/ControlRig.RigUnit_FABRIK(
    EffectorBone=cape_top_l_10
    MaxIterations=10
    Precision=1
    Weight=1
    bPropagateToChildren=True
    bSetEffectorTransform=True
    StartBone=&RigUnit_BoneName(Bone=cape_top_l_1).Bone
    EffectorTransform=@RigUnit_GetBoneTransform.Transform
   ) {is=struct RigVMUnitNode}
   > &RigUnit_FABRIK(
    EffectorBone=cape_mid_l_10
    MaxIterations=10
    Precision=1
    StartBone=cape_mid_l_1
    Weight=1
    bPropagateToChildren=True
    bSetEffectorTransform=True
    EffectorTransform=@RigUnit_GetBoneTransform.Transform
   )
   > &RigUnit_FABRIK(
    EffectorBone=cape_bot_l_10
    MaxIterations=10
    Precision=1
    StartBone=cape_bot_l_1
    Weight=1
    bPropagateToChildren=True
    bSetEffectorTransform=True
    EffectorTransform=@RigUnit_GetBoneTransform.Transform
   )
   n5 = &RigUnit_GetBoneTransform(Bone=attach_r)
   > &RigUnit_FABRIK(
    EffectorBone=cape_bot_r_10
    MaxIterations=10
    Precision=1
    StartBone=cape_bot_r_1
    Weight=1
    bPropagateToChildren=True
    bSetEffectorTransform=True
    EffectorTransform=@n5.Transform
   )
   n6 = &RigUnit_BoneName(Bone=cape_pv_aim_l)
   n7 = /Script/ControlRig.RigUnit_GetInitialBoneTransform(Space=GlobalSpace, Bone=@n6.Bone) {is=struct RigVMUnitNode}
   n8 = &RigUnit_GetInitialBoneTransform(Space=GlobalSpace, Bone=&RigUnit_BoneName(Bone=attach_l).Bone)
   n9 = &RigUnit_GetInitialBoneTransform(Space=GlobalSpace, Bone=&RigUnit_BoneName(Bone=lowerarm_l).Bone)
   &RigUnit_AimBone(
    DebugSettings=(bEnabled=true,Scale=10,WorldOffset=(Rotation=(X=0,Y=0,Z=0,W=1),Translation=(X=0,Y=0,Z=0),Scale3D=(X=1,Y=1,Z=1)))
    ExecutePin=()
    Primary=(Weight=1,Axis=(X=1,Y=0,Z=0),Target=(X=1,Y=0,Z=0),Kind=Location)
    Secondary=(Weight=1,Axis=(X=0,Y=0,Z=1),Target=(X=1,Y=0,Z=0),Kind=Location)
    Weight=1
    bPropagateToChildren=True
    Bone=@n6.Bone
    Primary.Axis=/Script/RigVM.RigVMFunction_MathVectorUnit(Value=&RigVMFunction_MathTransformMakeRelative(Parent=@n7.Transform, Global=@n8.Transform).Local.Translation) {is=struct RigVMUnitNode, template="Unit::Execute(in Value,out Result)"}.Result
    Secondary.Axis=&RigVMFunction_MathVectorUnit(Value=&RigVMFunction_MathTransformMakeRelative(Parent=@n7.Transform, Global=@n9.Transform).Local.Translation) {template="Unit::Execute(in Value,out Result)"}.Result
   )
   &RigUnit_FABRIK(
    EffectorBone=cape_top_r_10
    ExecutePin=()
    MaxIterations=10
    Precision=1
    StartBone=cape_top_r_1
    Weight=1
    bPropagateToChildren=True
    bSetEffectorTransform=True
    EffectorTransform=@n5.Transform
   )
   &RigUnit_FABRIK(
    EffectorBone=cape_mid_r_10
    ExecutePin=()
    MaxIterations=10
    Precision=1
    StartBone=cape_mid_r_1
    Weight=1
    bPropagateToChildren=True
    bSetEffectorTransform=True
    EffectorTransform=@n5.Transform
   )
   n10 = /Script/RigVM.RigVMFunction_MathVectorLerp(
    T=0.5
    A=&RigUnit_GetBoneTransform(Bone=upperarm_l).Transform.Translation
    B=&RigUnit_GetBoneTransform(Bone=hand_l).Transform.Translation
   ) {is=struct RigVMUnitNode, template="Interpolate::Execute(in A,in B,in T,out Result)"}
   /Script/RigVM.RigVMFunction_MathVectorAdd(
    B=&RigVMFunction_MathVectorUnit(Value=/Script/RigVM.RigVMFunction_MathVectorSub(B=@n10.Result, A=@n3.Transform.Translation) {is=struct RigVMUnitNode, template="Subtract::Execute(in A,in B,out Result)"}.Result) {template="Unit::Execute(in Value,out Result)"}.Result
    A=&RigUnit_GetBoneTransform(Bone=@n2.Bone).Transform.Translation
   ) {is=struct RigVMUnitNode, template="Add::Execute(in A,in B,out Result)"}
   &RigVMFunction_MathVectorAdd(A=(X=33,Y=-14,Z=0), B=(X=0,Y=0,Z=0)) {template="Add::Execute(in A,in B,out Result)"}
   /Script/RigVM.RigVMFunction_MathVectorMul(A=(X=1,Y=1,Z=1), B=(X=1,Y=1,Z=1)) {is=struct RigVMUnitNode, template="Multiply::Execute(in A,in B,out Result)"}
   /Script/RigVM.RigVMFunction_MathVectorScale(Factor=1, Value=(X=33,Y=-14,Z=0)) {is=struct RigVMUnitNode, template="Scale::Execute(in Value,in Factor,out Result)"}
  ]}
]
hierarchy [
 root {loc=<0 0 0>, rot=<0 0 0 1>}
 pelvis {parent=root, loc=<0 -0.5301 98.0571>, rot=<0.0127 0.707 -0.0127 -0.707>}
 spine_01 {parent=pelvis, loc=<6.1208 -0.9328 0>, rot=<0 0 0.034 -0.9994>}
 spine_02 {parent=spine_01, loc=<6.5996 -0.2095 0>, rot=<0 0 0.0238 -0.9997>}
 spine_03 {parent=spine_02, loc=<6.4914 -0.1237 0>, rot=<0 0 0.0149 -0.9999>}
 spine_04 {parent=spine_03, loc=<6.9598 0.0794 0>, rot=<0 0 -0.0132 -0.9999>}
 spine_05 {parent=spine_04, loc=<6.9987 0 0>, rot=<0 0 -0.0015 -1>}
 clavicle_l {parent=spine_05, loc=<5.483 -0.1053 -3.2946>, rot=<0.0465 -0.7289 -0.009 -0.6829>}
 upperarm_l {parent=clavicle_l, loc=<9.8658 -1.2136 -0.4105>, rot=<0.0164 -0.3245 -0.0269 -0.9454>}
 lowerarm_l {parent=upperarm_l, loc=<26.465 -0.8345 0.259>, rot=<0.0036 0.0161 -0.2298 -0.9731>}
 hand_l {parent=lowerarm_l, loc=<22.7355 -0.3203 -0.6105>, rot=<-0.6784 -0.0223 0.0735 -0.7307>}
 lowerarm_twist_01_l {parent=lowerarm_l, loc=<18.316 0 0>, rot=<0 0 0 -1>}
 Wrist_l {parent=lowerarm_l, loc=<18.3164 -0.0001 -0.0003>, rot=<0 -1 0 0>}
 attach_l {parent=Wrist_l, loc=<-2.4772 -4.1377 1.0431>, rot=<0.4877 -0.5104 0.5356 -0.4634>}
 dyn_cape_01_prox_l_MSTR {parent=attach_l, loc=<0 0 0>, rot=<0.4939 -0.5323 0.0903 -0.6816>}
 dyn_cape_01_prox_l {parent=dyn_cape_01_prox_l_MSTR, loc=<0.2001 0 0>, rot=<0 0 0 -1>}
 dyn_cape_02_prox_l {parent=dyn_cape_01_prox_l, loc=<12.0145 0 -0.2539>, rot=<0.0025 -0.0484 -0.0001 -0.9988>}
 dyn_cape_03_prox_l {parent=dyn_cape_02_prox_l, loc=<13.145 -0.0638 -0.0883>, rot=<-0.0828 -0.0158 0.0135 -0.9964>}
 cape_pv_up_l {parent=upperarm_l, loc=<43.6207 -9.7209 -24.9467>, rot=<-0.0065 0.3563 0.0039 -0.9343>}
 clavicle_r {parent=spine_05, loc=<5.4831 -0.1053 3.2946>, rot=<0.7289 0.0465 -0.6829 0.009>}
 upperarm_r {parent=clavicle_r, loc=<-9.8658 1.2136 0.4104>, rot=<0.0164 -0.3245 -0.0269 -0.9454>}
 lowerarm_r {parent=upperarm_r, loc=<-26.4651 0.8345 -0.259>, rot=<0.0036 0.0161 -0.2298 -0.9731>}
 hand_r {parent=lowerarm_r, loc=<-22.7356 0.3202 0.6105>, rot=<-0.6784 -0.0223 0.0735 -0.7307>}
 lowerarm_twist_01_r {parent=lowerarm_r, loc=<-18.3161 0 0>, rot=<0 0 0 -1>}
 Wrist_r {parent=lowerarm_r, loc=<-18.3161 0 0>, rot=<0 0 0 -1>}
 attach_r {parent=Wrist_r, loc=<-2.4775 4.1378 1.0434>, rot=<0.4877 0.5104 0.5356 0.4634>}
 dyn_cape_01_prox_r_MSTR {parent=attach_r, loc=<0 0 0>, rot=<-0.4939 -0.5323 -0.0903 -0.6816>}
 dyn_cape_01_prox_r {parent=dyn_cape_01_prox_r_MSTR, loc=<0.2 0 0>, rot=<0 0 0 -1>}
 dyn_cape_02_prox_r {parent=dyn_cape_01_prox_r, loc=<12.0146 0 -0.2539>, rot=<-0.0025 -0.0484 0.0001 -0.9988>}
 dyn_cape_03_prox_r {parent=dyn_cape_02_prox_r, loc=<13.1451 0.0638 -0.0883>, rot=<0.0828 -0.0158 -0.0135 -0.9964>}
 cape_pv_up_r {parent=upperarm_r, loc=<-43.6208 9.7209 24.9468>, rot=<-0.3563 -0.0065 0.9343 0.0039>}
 backpack_root {parent=spine_05, loc=<3.5758 -5.2058 0>, rot=<0.0284 -0.7065 -0.0284 -0.7065>}
 cape_all {parent=backpack_root, loc=<-0.0182 -2.7051 2.4791>, rot=<-0.4644 -0.4613 0.5332 -0.536>}
 cape_pv_aim_l {parent=cape_all, loc=<-1.4249 11.4405 -0.0088>, rot=<0.8708 0.4239 -0.0552 0.2428>}
 cape_top_l_1 {parent=cape_pv_aim_l, loc=<0 0 0>, rot=<0.8765 0.4109 -0.0615 -0.2432>}
 cape_top_l_2 {parent=cape_top_l_1, loc=<15.7852 0.4377 0.2636>, rot=<0.0005 0.0124 0.0049 -0.9999>}
 cape_top_l_3 {parent=cape_top_l_2, loc=<10.5642 0.3216 -0.9065>, rot=<-0.0102 0.0143 -0.0557 -0.9983>}
 cape_top_l_4 {parent=cape_top_l_3, loc=<11.1188 -0.1865 -0.6094>, rot=<-0.0223 -0.0274 -0.1278 -0.9912>}
 cape_top_l_5 {parent=cape_top_l_4, loc=<9.1424 0.6001 -0.4043>, rot=<-0.0038 -0.1239 -0.2055 -0.9708>}
 dyn_cape_A_01_l_MSTR {parent=cape_top_l_5, loc=<1.0321 -0.7716 1.6086>, rot=<0.0634 0.0337 0.3875 -0.9191>}
 dyn_cape_A_01_l {parent=dyn_cape_A_01_l_MSTR, loc=<0.2 0 -0.0001>, rot=<0 0 0 -1>}
 dyn_cape_A_02_l {parent=dyn_cape_A_01_l, loc=<13.9552 0 0.0001>, rot=<0 0 0 -1>}
 dyn_cape_A_03_l {parent=dyn_cape_A_02_l, loc=<15.527 0 -0.0001>, rot=<0 0 0 -1>}
 dyn_cape_A_04_l {parent=dyn_cape_A_03_l, loc=<15.4412 0 0>, rot=<0 0 0 -1>}
 cape_top_l_6 {parent=cape_top_l_5, loc=<7.5415 1.2154 0.2635>, rot=<0.066 -0.121 -0.2755 -0.9514>}
 cape_top_l_7 {parent=cape_top_l_6, loc=<8.9478 0.1241 0.7512>, rot=<0.1286 -0.0503 -0.2922 -0.9463>}
 dyn_cape_01_l_MSTR {parent=cape_top_l_7, loc=<4.7191 -2.2012 -0.4806>, rot=<-0.0144 -0.1729 0.8133 -0.5554>}
 dyn_cape_01_l {parent=dyn_cape_01_l_MSTR, loc=<0.2001 0 0>, rot=<0 0 0 -1>}
 dyn_cape_02_l {parent=dyn_cape_01_l, loc=<11.656 0 0>, rot=<0 0 0 -1>}
 dyn_cape_03_l {parent=dyn_cape_02_l, loc=<13.2284 0 0>, rot=<0 0 0 -1>}
 dyn_cape_04_l {parent=dyn_cape_03_l, loc=<13.1417 -0.0001 0>, rot=<0 0 0 -1>}
 cape_top_l_8 {parent=cape_top_l_7, loc=<8.5143 -1.9988 0.2644>, rot=<0.1148 0.0509 -0.2608 -0.9572>}
 cape_top_l_9 {parent=cape_top_l_8, loc=<9.4974 -2.0622 -0.9136>, rot=<0.103 -0.0222 -0.2546 -0.9613>}
 cape_top_l_10 {parent=cape_top_l_9, loc=<10.3953 -0.9907 -1.8512>, rot=<0 0 0 -1>}
 cape_pv_aim_r {parent=cape_all, loc=<-1.5616 -11.4227 0.0093>, rot=<0.8685 -0.4291 -0.0535 -0.2425>}
 cape_top_r_1 {parent=cape_pv_aim_r, loc=<0 0 0>, rot=<0.8765 -0.4109 -0.0615 0.2432>}
 cape_top_r_2 {parent=cape_top_r_1, loc=<15.7856 -0.4377 0.2636>, rot=<-0.0005 0.0124 -0.0049 -0.9999>}
 cape_top_r_3 {parent=cape_top_r_2, loc=<10.5642 -0.3216 -0.9066>, rot=<0.0102 0.0143 0.0557 -0.9983>}
 cape_top_r_4 {parent=cape_top_r_3, loc=<11.1189 0.1865 -0.6095>, rot=<0.0223 -0.0274 0.1278 -0.9912>}
 cape_top_r_5 {parent=cape_top_r_4, loc=<9.1421 -0.6002 -0.4042>, rot=<0.0038 -0.1239 0.2055 -0.9708>}
 dyn_cape_A_01_r_MSTR {parent=cape_top_r_5, loc=<1.0059 0.7966 1.6057>, rot=<-0.0634 0.0337 -0.3875 -0.9191>}
 dyn_cape_A_01_r {parent=dyn_cape_A_01_r_MSTR, loc=<0.2 0 0>, rot=<0 0 0 1>}
 dyn_cape_A_02_r {parent=dyn_cape_A_01_r, loc=<13.9552 0 0>, rot=<0 0 0 1>}
 dyn_cape_A_03_r {parent=dyn_cape_A_02_r, loc=<15.527 0 0>, rot=<0 0 0 1>}
 dyn_cape_A_04_r {parent=dyn_cape_A_03_r, loc=<15.4412 0 0>, rot=<0 0 0 1>}
 cape_top_r_6 {parent=cape_top_r_5, loc=<7.5415 -1.2155 0.2636>, rot=<-0.066 -0.121 0.2755 -0.9514>}
 cape_top_r_7 {parent=cape_top_r_6, loc=<8.9478 -0.124 0.7511>, rot=<-0.1286 -0.0503 0.2922 -0.9463>}
 dyn_cape_01_r_MSTR {parent=cape_top_r_7, loc=<4.6901 2.1848 -0.4952>, rot=<0.0144 -0.1729 -0.8133 -0.5554>}
 dyn_cape_01_r {parent=dyn_cape_01_r_MSTR, loc=<0.2 0 0>, rot=<0 0 0 1>}
 dyn_cape_02_r {parent=dyn_cape_01_r, loc=<11.656 0 0>, rot=<0 0 0 1>}
 dyn_cape_03_r {parent=dyn_cape_02_r, loc=<13.2285 0 0>, rot=<0 0 0 1>}
 dyn_cape_04_r {parent=dyn_cape_03_r, loc=<13.1417 0 0>, rot=<0 0 0 1>}
 cape_top_r_8 {parent=cape_top_r_7, loc=<8.5142 1.9989 0.2645>, rot=<-0.1148 0.0509 0.2608 -0.9572>}
 cape_top_r_9 {parent=cape_top_r_8, loc=<9.4975 2.0622 -0.9136>, rot=<-0.103 -0.0222 0.2546 -0.9613>}
 cape_top_r_10 {parent=cape_top_r_9, loc=<10.3953 0.9906 -1.8512>, rot=<0 0 0 -1>}
 dyn_cape_mid_01_MSTR {parent=cape_all, loc=<2.6887 0 0>, rot=<0 0.0135 0 -0.9999>}
 dyn_cape_mid_01 {parent=dyn_cape_mid_01_MSTR, loc=<5.7358 0 0>, rot=<0 0 0 -1>}
 dyn_cape_mid_02 {parent=dyn_cape_mid_01, loc=<15.1927 0 0>, rot=<0 0 0 -1>}
 dyn_cape_mid_03 {parent=dyn_cape_mid_02, loc=<16.764 0 0>, rot=<0 0 0 -1>}
 dyn_cape_mid_04 {parent=dyn_cape_mid_03, loc=<16.6788 0 0>, rot=<0 0 0 -1>}
 dyn_cape_mid_05 {parent=dyn_cape_mid_04, loc=<18.9955 0 0>, rot=<0 0 0 -1>}
 cape_clavicle_l {parent=cape_all, loc=<-1.0021 3.3123 -7.5847>, rot=<0.4699 0.4541 -0.504 -0.5648>}
 cape_clavicle_r {parent=cape_all, loc=<-1.0415 -3.2768 -7.5795>, rot=<-0.507 -0.5622 -0.4667 -0.4573>}
 thigh_l {parent=pelvis, loc=<-1.0855 -0.2109 -8.7053>, rot=<0.0009 0.0355 0.0155 -0.9993>}
 calf_l {parent=thigh_l, loc=<-42.5284 -0.0019 -0.7297>, rot=<0.0006 -0.0161 -0.0391 -0.9991>}
 foot_l {parent=calf_l, loc=<-46.7401 -2.5014 -0.592>, rot=<0.0427 -0.0238 0.0353 -0.9982>}
 thigh_twist_01_l {parent=thigh_l, loc=<-27.952 0 0>, rot=<0.0013 0 0 -1>}
 thigh_r {parent=pelvis, loc=<-1.0853 -0.2109 8.7053>, rot=<0.0355 -0.0009 0.9993 0.0155>}
 calf_r {parent=thigh_r, loc=<42.5286 0.0019 0.7297>, rot=<0.0006 -0.0161 -0.0391 -0.9991>}
 foot_r {parent=calf_r, loc=<46.7403 2.5014 0.592>, rot=<0.0427 -0.0238 0.0353 -0.9982>}
 thigh_twist_01_r {parent=thigh_r, loc=<27.952 0 0>, rot=<0.0013 0 0 -1>}
 attach_tgt_l {parent=pelvis, loc=<0 0 0>, rot=<0.5457 -0.725 -0.4174 0.0489>}
 attach_tgt_r {parent=pelvis, loc=<0 0 0>, rot=<0.5457 -0.725 0.4174 -0.0489>}
 body_root {parent=root, loc=<0 0 0>, rot=<0 0 0 1>}
 floorCOL {parent=root, loc=<0 0 0>, rot=<0 0 0 1>}
]
variables [
 offset_l {type=double, default=0}
 offset_r {type=double, default=0}
]
