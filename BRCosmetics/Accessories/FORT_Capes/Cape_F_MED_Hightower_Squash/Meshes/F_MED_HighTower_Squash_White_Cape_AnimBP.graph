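# Animation Blueprint for the Hightower Squash white cape cosmetic, posing and simulating it on the wearer.
# The copied parent pose feeds a FABRIK Control Rig, bone constraints, then three RigidBody simulations for body and arms.
# When rigid bodies are disabled, additive bone rotations and look-at nodes adjust that cached result.
# The EventGraph sets a variant-tag alpha and holds noise-driven gravity logic for swimming and skydiving, disconnected from the update event.
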AnimBlueprint /BRCosmetics/Accessories/FORT_Capes/Cape_F_MED_Hightower_Squash/Meshes/F_MED_HighTower_Squash_White_Cape_AnimBP
target_skeleton=&/F_MED_HighTower_Squash_White_Cape_Skeleton
graphs [
 AnimGraph {nodes [
   n1 = "Constraint - dyn_cape_mid_01 (R)"(Alpha=1, ConstraintWeights_0=1) {
    alphaBoolBlend {blendOption=Linear}
    boneToModify {boneName=dyn_cape_mid_01}
    constraintSetup [{transformType=Rotation, targetBone {boneName=spine_04}}]
    constraintWeights [1]
   }
   n2 = "Constraint - dyn_cape_mid_02 (R)"(ComponentPose=@n1, Alpha=1, ConstraintWeights_0=1) {
    alphaBoolBlend {blendOption=Linear}
    boneToModify {boneName=dyn_cape_mid_02}
    constraintSetup [{transformType=Rotation, targetBone {boneName=spine_02}}]
    constraintWeights [1]
   }
   n3 = "Constraint - dyn_cape_mid_03 (PP)"!(ComponentPose=@n2, Alpha=1, ConstraintWeights_0=1, ConstraintWeights_1:float=1) -> (Pose) {
    alphaBoolBlend {blendOption=Linear}
    boneToModify {boneName=dyn_cape_mid_03}
    constraintSetup [{targetBone {boneName=thigh_l}}, {targetBone {boneName=thigh_r}}]
    constraintWeights [1,1]
   }
   "Constraint - dyn_cape_mid_04 (PP)"!(ComponentPose=@n3, Alpha=1, ConstraintWeights_0=1, ConstraintWeights_1:float=1) -> (Pose) {
    alphaBoolBlend {blendOption=Linear}
    boneToModify {boneName=dyn_cape_mid_04}
    constraintSetup [{targetBone {boneName=calf_l}}, {targetBone {boneName=calf_r}}]
    constraintWeights [1,1]
   }
   n4 = "Transform (Modify) Bone - Bone: dyn_cape_mid_01"(
    ComponentPose=Local To Component(LocalPose=Use cached pose 'main'())
    Alpha=1
    Translation="0,0,0"
    Rotation="25.326,0.000,0.000"
    Scale="1,1,1"
   ) {
    rotationMode=BMM_Additive
    rotationSpace=BCS_BoneSpace
    alphaBoolBlend {blendOption=Linear}
    boneToModify {boneName=dyn_cape_mid_01}
    rotation {pitch=25.326}
   }
   n5 = "Transform (Modify) Bone - Bone: dyn_cape_mid_02"(ComponentPose=@n4, Alpha=1, Translation="0,0,0", Rotation="11.917,0.000,0.000", Scale="1,1,1") {
    rotationMode=BMM_Additive
    rotationSpace=BCS_BoneSpace
    alphaBoolBlend {blendOption=Linear}
    boneToModify {boneName=dyn_cape_mid_02}
    rotation {pitch=11.917}
   }
   n6 = "Transform (Modify) Bone - Bone: dyn_cape_mid_03"(ComponentPose=@n5, Alpha=1, Translation="0,0,0", Rotation="7.579,0.000,0.000", Scale="1,1,1") {
    rotationMode=BMM_Additive
    rotationSpace=BCS_BoneSpace
    alphaBoolBlend {blendOption=Linear}
    boneToModify {boneName=dyn_cape_mid_03}
    rotation {pitch=7.579}
   }
   n7 = "Transform (Modify) Bone - Bone: dyn_cape_mid_04"(ComponentPose=@n6, Alpha=1, Translation="0,0,0", Rotation="10.999,0.000,0.000", Scale="1,1,1") {
    rotationMode=BMM_Additive
    rotationSpace=BCS_BoneSpace
    alphaBoolBlend {blendOption=Linear}
    boneToModify {boneName=dyn_cape_mid_04}
    rotation {pitch=10.999}
   }
   n8 = "Transform (Modify) Bone - Bone: dyn_cape_A_01_l_MSTR"(ComponentPose=@n7, Alpha=1, Translation="0,0,0", Rotation="21.173,0.000,0.000", Scale="1,1,1") {
    rotationMode=BMM_Additive
    rotationSpace=BCS_BoneSpace
    alphaBoolBlend {blendOption=Linear}
    boneToModify {boneName=dyn_cape_A_01_l_MSTR}
    rotation {pitch=21.173}
   }
   n9 = "Transform (Modify) Bone - Bone: dyn_cape_A_01_r_MSTR"(ComponentPose=@n8, Alpha=1, Translation="0,0,0", Rotation="30.821,0.000,0.000", Scale="1,1,1") {
    rotationMode=BMM_Additive
    rotationSpace=BCS_BoneSpace
    alphaBoolBlend {blendOption=Linear}
    boneToModify {boneName=dyn_cape_A_01_r_MSTR}
    rotation {pitch=30.821}
   }
   n10 = Look At - Bone: dyn_cape_01_prox_r_MSTR(ComponentPose=@n9, Alpha=1) {
    alphaBoolBlend {blendOption=Linear}
    boneToModify {boneName=dyn_cape_01_prox_r_MSTR}
    lookAtLocation {x=-70, z=50}
    lookAtTarget {boneReference {boneName=pelvis}}
    lookAt_Axis {axis=<1 0>}
   }
   n11 = Look At - Bone: dyn_cape_01_prox_l_MSTR(ComponentPose=@n10, Alpha=1) {
    alphaBoolBlend {blendOption=Linear}
    boneToModify {boneName=dyn_cape_01_prox_l_MSTR}
    lookAtLocation {x=-70, z=-50}
    lookAtTarget {boneReference {boneName=pelvis}}
    lookAt_Axis {axis=<1 0>}
   }
   Output Pose(Result=Blend Poses by bool(BlendPose_0=Use cached pose 'main'(), BlendPose_1=Component To Local(ComponentPose=@n11), bActiveValue=Get bRigidBodyEnabled()) {blendType=Linear}) {layerGroup=None}
   n12 = Control Rig!(
    Source=Copy Pose From Mesh() {bUseAttachedParent=true}
    Alpha=Get fabrik_alpha() -> (fabrik_alpha:double)
    offset_r:double=Get offset_r() -> (offset_r:double)
    offset_l:double=Get offset_l() -> (offset_l:double)
   ) -> (Pose) {
    alphaBoolBlend {blendOption=Linear}
    controlRigAssetReference {blueprintRigClass=&/Hightower_Squash_Fabrik_CtrlRig.Hightower_Squash_Fabrik_CtrlRig_C}
   }
   n13 = "Constraint - cape_clavicle_l (R)"(ComponentPose=Local To Component(LocalPose=@n12), Alpha=1, ConstraintWeights_0=1) {
    alphaBoolBlend {blendOption=Linear}
    boneToModify {boneName=cape_clavicle_l}
    constraintSetup [{transformType=Rotation, targetBone {boneName=clavicle_l}}]
    constraintWeights [1]
   }
   n14 = "Constraint - cape_clavicle_r (R)"(ComponentPose=@n13, Alpha=1, ConstraintWeights_0=1) {
    alphaBoolBlend {blendOption=Linear}
    boneToModify {boneName=cape_clavicle_r}
    constraintSetup [{transformType=Rotation, targetBone {boneName=clavicle_r}}]
    constraintWeights [1]
   }
   n15 = "Constraint - dyn_cape_A_01_r_MSTR (PP)"!(ComponentPose=@n14, Alpha=1, ConstraintWeights_0=0.75, ConstraintWeights_1:float=0.25) -> (Pose) {
    alphaBoolBlend {blendOption=Linear}
    boneToModify {boneName=dyn_cape_A_01_r_MSTR}
    constraintSetup [{targetBone {boneName=dyn_cape_mid_03}}, {targetBone {boneName=cape_top_r_5}}]
    constraintWeights [1,1]
   }
   n16 = "Constraint - dyn_cape_01_r_MSTR (PP)"!(ComponentPose=@n15, Alpha=1, ConstraintWeights_0=0.5, ConstraintWeights_1:float=0.5) -> (Pose) {
    alphaBoolBlend {blendOption=Linear}
    boneToModify {boneName=dyn_cape_01_r_MSTR}
    constraintSetup [{targetBone {boneName=dyn_cape_A_01_r_MSTR}}, {targetBone {boneName=cape_top_r_7}}]
    constraintWeights [1,1]
   }
   n17 = "Constraint - dyn_cape_A_01_l_MSTR (PP)"!(ComponentPose=@n16, Alpha=1, ConstraintWeights_0=0.75, ConstraintWeights_1:float=0.25) -> (Pose) {
    alphaBoolBlend {blendOption=Linear}
    boneToModify {boneName=dyn_cape_A_01_l_MSTR}
    constraintSetup [{targetBone {boneName=dyn_cape_mid_03}}, {targetBone {boneName=cape_top_l_5}}]
    constraintWeights [1,1]
   }
   n18 = "Constraint - dyn_cape_01_l_MSTR (PP)"!(ComponentPose=@n17, Alpha=1, ConstraintWeights_0=0.5, ConstraintWeights_1:float=0.5) -> (Pose) {
    alphaBoolBlend {blendOption=Linear}
    boneToModify {boneName=dyn_cape_01_l_MSTR}
    constraintSetup [{targetBone {boneName=dyn_cape_A_01_l_MSTR}}, {targetBone {boneName=cape_top_l_7}}]
    constraintWeights [1,1]
   }
   n19 = RigidBody!(ComponentPose=@n18, Alpha, OverrideWorldGravity:Vector=Get GravityOverride(), ExternalForce) -> (Pose) {
    bClampLinearTranslationLimitToRefPose=true
    bOverrideWorldGravity=true
    overridePhysicsAsset=&/F_MED_HighTower_Squash_White_Cape_Dyn
    simulationSpace=BaseBoneSpace
    alphaBoolBlend {blendOption=Linear}
    baseBoneRef {boneName=spine_05}
    componentAppliedLinearAccClamp=<2000 3000 3000>
    componentLinearAccScale=<0.1 0.35 0.35>
    componentLinearVelScale=<0.1 0.35 0.35>
    overrideWorldGravity {z=-980}
    simSpaceSettings {maxAngularAcceleration=500, maxAngularVelocity=500, velocityScaleZ=0, worldAlpha=0.05, externalLinearDragV=<0.1 0.1 0.1>}
   }
   n20 = RigidBody(ComponentPose=@n19, OverrideWorldGravity="0,0,-980") {
    bClampLinearTranslationLimitToRefPose=true
    bOverrideWorldGravity=true
    overridePhysicsAsset=&/F_MED_HighTower_Squash_White_Cape_Dyn_arm_l
    simulationSpace=BaseBoneSpace
    alphaBoolBlend {blendOption=Linear}
    baseBoneRef {boneName=upperarm_l}
    componentAppliedLinearAccClamp=<3000 3000 3000>
    overrideWorldGravity {z=-980}
    simSpaceSettings {velocityScaleZ=0, worldAlpha=0.1, externalLinearDragV=<0.1 0.1 0.1>}
   }
   n21 = RigidBody(ComponentPose=@n20, OverrideWorldGravity="0,0,-980") {
    bClampLinearTranslationLimitToRefPose=true
    bOverrideWorldGravity=true
    overridePhysicsAsset=&/F_MED_HighTower_Squash_White_Cape_Dyn_arm_r
    simulationSpace=BaseBoneSpace
    alphaBoolBlend {blendOption=Linear}
    baseBoneRef {boneName=upperarm_r}
    componentAppliedLinearAccClamp=<3000 3000 3000>
    overrideWorldGravity {z=-980}
    simSpaceSettings {velocityScaleZ=0, worldAlpha=0.1, externalLinearDragV=<0.1 0.1 0.1>}
   }
   Save cached pose 'main'(Pose=Component To Local(ComponentPose=@n21)) {cacheName=main}
  ]}
 EventGraph {nodes [
   Event BlueprintInitializeAnimation()
   > Set Skel_Mesh(Skel_Mesh=GetOwnerSkeletalMeshComponent())
   Event OnCharacterCustomizationLoaded()
   n1 = Make Array!(
    "[0]"=(TagName="Cosmetics.Variant.Property.Stage1")
    "[1]":Gameplay Tag Structure=(TagName="Cosmetics.Variant.Property.Stage2")
    "[2]":Gameplay Tag Structure=(TagName="Cosmetics.Variant.Property.Mat2")
   ) -> (Array)
   > Set punk_alpha(punk_alpha:double="To Float (Boolean)"(InBool=CharacterItemDefinitionHasAnyMetaTag(Tags=@n1))) -> (Output_Get:double)
   Sequence()
   > n2 = Branch(Condition=Get bIsSwimSprinting() -> (bIsSwimSprinting:bool))
   n3 = Get Skel_Mesh()
   n4 = GetSocketRotation!(self:Scene Component Object Reference=@n3, InSocketName:Name=pelvis) -> (ReturnValue:Rotator)
   n5 = PerlinNoise1D!(Value:float=GetGameTimeInSeconds()) -> (ReturnValue:float)
   n6 = float * float(A="Sin (Degrees)"!(A:double=float * float(A=@n5, B=800)) -> (ReturnValue:double), B=-400)
   n7 = vector + vector(
    A:Vector=vector * float(A=GetForwardVector!(InRot:Rotator=@n4) -> (ReturnValue:Vector), B=-300)
    B:Vector=MakeVector(Y=0, Z=float + float(A=@n6, B=200) {is=title K2Node_CommutativeAssociativeBinaryOperator})
   ) -> (ReturnValue:Vector) {is=title K2Node_CommutativeAssociativeBinaryOperator}
   > Set GravityOverride_b(GravityOverride_b:Vector=@n7) -> (Output_Get:Vector)
   n8 = Branch(execute=@n2.else, Condition=Get bIsSkyDiving())
   > n9 = Branch(Condition=Get bIsParachuteOpened())
   n10 = float * float(A=@n5, B=800)
   n11 = float + float(A=float * float(A="Sin (Degrees)"(A=@n10), B=-400), B=-600)
   n12 = GetRightVector!(InRot:Rotator=@n4) -> (ReturnValue:Vector)
   n13 = Set GravityOverride_b(execute=[@n8.else, @n9.then], GravityOverride_b=vector + vector(A=vector * float(A=@n12, B=-400), B=MakeVector(Y=0, Z=@n11)))
   > Set GravityOverrideArms_b(GravityOverrideArms_b:Vector=vector * float(A=@n13, B=0.3)) -> (Output_Get:Vector)
   n14 = Branch(execute=@n9.else, Condition="float > float"(A=Get PawnVelocityZ(), B=50) {is=title K2Node_CallFunction})
   > Set GravityOverride_b(GravityOverride_b="0, 0,-100")
   Set GravityOverride_b(execute=@n14.else, GravityOverride_b=@n7)
   n15 = GetSocketLocation!(self:Scene Component Object Reference=@n3, InSocketName:Name=pelvis) -> (ReturnValue:Vector)
   DrawDebugLine!(
    execute
    LineStart:Vector=@n15
    LineEnd:Vector=vector + vector(A=@n15, B="0, 0, 0")
    LineColor:Linear Color Structure=(R=1,G=0,B=0,A=1)
    Duration:float=0
    Thickness:float=3
    DepthPriority:EDrawDebugSceneDepthPriorityGroup Enum=World
   ) -> (then)
   Event BlueprintUpdateAnimation()
  ]}
]
variables [
 fabrik_alpha {type=number, default=1}
 GravityOverride_b {type=Vector}
 GravityOverrideArms_b {type=Vector}
 offset_l {type=number, default=1}
 offset_r {type=number, default=1}
 punk_alpha {type=number}
 Skel_Mesh {type=SkeletalMeshComponent}
 SkelMesh {type=SkeletalMeshComponent}
]
